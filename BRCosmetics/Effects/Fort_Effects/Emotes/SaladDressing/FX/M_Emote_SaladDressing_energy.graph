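# Translucent, two-sided unlit Niagara material for the SaladDressing emote's energy effect, usable on sprites, ribbons and meshes.
# Opacity combines a noise-warped edge mask with panning base, mask and optional caustic textures, then applies dissolve, view-angle, depth and particle-centered sphere fades.
# Emissive uses particle color or a mask-driven center-to-edge gradient.
# Dynamic parameters drive dissolve, UVs, speeds and width; scalar lerps toggle SplineThicken and texture-driven vertex wobble.
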Material /BRCosmetics/Effects/Fort_Effects/Emotes/SaladDressing/FX/M_Emote_SaladDressing_energy
graph {
 properties {
  bUsedWithNiagaraMeshParticles=true
  bUsedWithNiagaraRibbons=true
  bUsedWithNiagaraSprites=true
  blendMode=BLEND_Translucent
  pixelDepthOffsetMode=PDOM_Legacy
  shadingModel=MSM_Unlit
  twoSided=true
  usedShadingModels=Unlit
 }
 nodes [
  n1 = DynamicParameter() {defaultValue=<rgba -1 1 1 1>, paramNames [Dissolve, MaskTextureSpeedX_offset, MaskTextureSpeedY_offset, SplineWidth_Dynamic]}
  n2 = LinearInterpolate(A=Constant() {r=1}, B=@n1.SplineWidth_Dynamic, Alpha=ScalarParameter() {parameterName=Use_SplineWidthDynamic})
  n3 = Multiply(A=ScalarParameter() {defaultValue=3, parameterName=SplineBase}, B=@n2)
  n4 = Multiply(A=ScalarParameter() {defaultValue=3, parameterName=SplineTip}, B=@n2)
  n5 = StaticBool()
  n6 = TextureCoordinate() {vTiling=0}
  n7 = AppendVector(A=ScalarParameter() {defaultValue=1, parameterName=WPOTextureU}, B=ScalarParameter() {parameterName=WPOTextureV})
  n8 = ScalarParameter() {defaultValue=-2, parameterName=WPOTextureSpeedX}
  n9 = ScalarParameter() {parameterName=WPOTextureSpeedY}
  n10 = TextureSampleParameter2D(UVs=Panner(Coordinate=Multiply(A=@n6, B=@n7), Speed=AppendVector(A=@n8, B=@n9)) {speedX=-2}) {parameterName=WPOTexture, samplerType=SAMPLERTYPE_Normal, texture=&T_Weap_Distort_N_02}
  n11 = TextureCoordinate()
  n12 = Multiply(
   A=LinearInterpolate(Alpha=@n10.R) {constB=1.5}
   B=LinearInterpolate(A=ScalarParameter() {defaultValue=0.05, parameterName=WPOMask}, Alpha=ComponentMask(None=@n11) {r=true}) {constA=0.05}
  )
  n13 = LinearInterpolate(A=@n12, B=Multiply(A=VertexNormalWS(), B=@n12), Alpha=ScalarParameter() {parameterName=Use_VertexNormalMul})
  n14 = Multiply(A=Transform(None=Constant3Vector() {constant=<rgba 1 0 0 0>}) {transformSourceType=TRANSFORMSOURCE_Local}, B=@n13)
  n15 = ScalarParameter() {defaultValue=1, parameterName=X WPO_Intensity}
  n16 = Constant() {r=1}
  n17 = Multiply(A=Transform(None=Constant3Vector() {constant=<rgba 0 1 0 0>}) {transformSourceType=TRANSFORMSOURCE_Local}, B=@n13)
  n18 = ScalarParameter() {defaultValue=1, parameterName=Y WPO_Intensity}
  n19 = Add(A=Multiply(A=Multiply(A=@n14, B=@n15) {constB=10}, B=@n16), B=Multiply(A=@n16, B=Multiply(A=@n17, B=@n18) {constB=10}))
  n20 = Transform(None=Constant3Vector() {constant=<rgba 0 0 1 0>}) {transformSourceType=TRANSFORMSOURCE_Local}
  n21 = Multiply(A=@n16, B=Multiply(A=Multiply(A=@n20, B=@n13), B=ScalarParameter() {defaultValue=1, parameterName=Z WPO_Intensity}) {constB=10})
  Add(
   A=LinearInterpolate(
    A=MaterialFunctionCall(WidthBase=@n3, WidthTip=@n4, Expand U or V UV Channel=@n5) {materialFunction=&SplineThicken}.WorldPositionOffset
    B=Constant()
    Alpha=ScalarParameter() {defaultValue=1, parameterName=Use_SplineThicken}
   )
   B=LinearInterpolate(A=Constant(), B=Add(A=@n19, B=@n21), Alpha=ScalarParameter() {parameterName=Use_WPO})
  )
  n22 = ParticleColor()
  n23 = Panner(Coordinate=TextureCoordinate() {uTiling=2, vTiling=0.8}) {speedX=-2}
  n24 = Add(
   A=TextureCoordinate() {vTiling=0.45}
   B=Multiply(
    A=TextureSample(UVs=@n23) {samplerType=SAMPLERTYPE_LinearGrayscale, texture=/Game/Environments/Hexmap/HexClouds/T_HexmapFog_01_Height02}.R
    B=ScalarParameter() {defaultValue=0.03, parameterName=MaskNoise_Int}
   ) {constB=0.3}
  )
  n25 = ComponentMask(None=@n24) {r=true}
  n26 = ScalarParameter() {defaultValue=20, group=LinearMask, parameterName=MaskLeft_Mul}
  n27 = Saturate(None=Power(Base=Multiply(A=@n25, B=@n26) {constB=20}) {constExponent=1})
  n28 = OneMinus(None=@n25)
  n29 = ScalarParameter() {defaultValue=2.5, group=LinearMask, parameterName=MaskRight_Mul}
  n30 = Saturate(None=Power(Base=Multiply(A=@n28, B=@n29) {constB=2.5}) {constExponent=1})
  n31 = Sine(None=ComponentMask(None=@n24) {g=true})
  n32 = LinearInterpolate(
   A=ScalarParameter() {group=LinearMask, parameterName=LinearMaskLerp}
   Alpha=Multiply(A=Multiply(A=@n27, B=@n30), B=Power(Base=@n31) {constExponent=3})
  )
  n33 = DynamicParameter() {parameterIndex=2, defaultValue=<rgba 1.2 0.8 1 0.8>, paramNames [NoiseTextureU, NoiseTextureV, BaseTextureU, BaseTextureV]}
  n34 = Add(A=Multiply(A=TextureCoordinate(), B=AppendVector(A=@n33.BaseTextureU, B=@n33.BaseTextureU)), B=Constant2Vector() {r=0.15})
  n35 = DynamicParameter() {
   parameterIndex=3
   defaultValue=<rgba 1.2 0.8 -2 0>
   paramNames [NoiseTextureU_offset, NoiseTextureV_offset, NoiseTextureX_Speedoffset, NoiseTextureY_Speedoffset]
  }
  n36 = Add(A=TextureCoordinate(), B=AppendVector(A=@n35.NoiseTextureU_offset, B=@n35.NoiseTextureU_offset))
  n37 = DynamicParameter() {parameterIndex=2, defaultValue=<rgba 1.2 0.8 1 0.8>, paramNames [NoiseTextureU, NoiseTextureV, BaseTextureU, BaseTextureV]}
  n38 = AppendVector(A=@n37.NoiseTextureU, B=@n37.NoiseTextureU)
  n39 = TextureSampleParameter2D(UVs=Panner(Coordinate=Multiply(A=@n36, B=@n38), Speed=AppendVector(A=@n35.NoiseTextureX_Speedoffset, B=@n35.NoiseTextureX_Speedoffset)) {speedX=-2}) {group=Noise, parameterName=NoiseTexture, texture=/BRCosmetics/Weapons/FORT_Melee/Pickaxe_CoatCheck/FX/T_Pickaxe_CoatCheck_Noise}
  n40 = ScalarParameter() {defaultValue=0.1, group=Noise, parameterName=NoiseTexMaskLerp}
  n41 = ComponentMask(None=TextureCoordinate()) {r=true}
  n42 = Multiply(
   A=Multiply(A=ComponentMask(None=@n39.RGB) {g=true, r=true}, B=LinearInterpolate(A=@n40, Alpha=@n41) {constA=0.1})
   B=ScalarParameter() {defaultValue=0.4, group=Noise, parameterName=Noise_Intensity}
  ) {constB=0.3}
  n43 = DynamicParameter() {defaultValue=<rgba -1 1 1 1>, paramNames [Dissolve, MaskTextureSpeedX_offset, MaskTextureSpeedY_offset, SplineWidth_Dynamic]}
  n44 = Multiply(A=ScalarParameter() {defaultValue=-1, group=BaseTexture, parameterName=BaseTexture_SpeedX}, B=@n43.MaskTextureSpeedX_offset)
  n45 = Multiply(A=ScalarParameter() {defaultValue=-0.1, group=BaseTexture, parameterName=BaseTexture_SpeedY}, B=@n43.MaskTextureSpeedY_offset)
  n46 = TextureSampleParameter2D(UVs=Panner(Coordinate=Add(A=@n34, B=@n42), Speed=AppendVector(A=@n44, B=@n45)) {speedX=-1, speedY=-0.1}) {
   group=BaseTexture
   parameterName=Base_Texture
   samplerType=SAMPLERTYPE_LinearColor
   texture=&T_NoisePack_LinearColor
  }
  n47 = ChannelMaskParameter(None=@n46.RGBA) {group=BaseTexture, parameterName=Tex_ChargeRGBA, sortPriority=7}
  n48 = ScalarParameter() {defaultValue=1, group=MaskTexture, parameterName=MaskTextureV}
  n49 = DynamicParameter() {parameterIndex=1, paramNames [Param1, Param2, Param3, MaskTextureV_offset]}
  n50 = Multiply(
   A=TextureCoordinate()
   B=AppendVector(A=ScalarParameter() {defaultValue=1, group=MaskTexture, parameterName=MaskTextureU}, B=Multiply(A=@n48, B=@n49.MaskTextureV_offset))
  )
  n51 = DynamicParameter() {defaultValue=<rgba -1 1 1 1>, paramNames [Dissolve, MaskTextureSpeedX_offset, MaskTextureSpeedY_offset, SplineWidth_Dynamic]}
  n52 = Multiply(A=ScalarParameter() {defaultValue=-1, group=MaskTexture, parameterName=MaskTexture_SpeedX}, B=@n51.MaskTextureSpeedX_offset)
  n53 = Multiply(A=ScalarParameter() {group=MaskTexture, parameterName=MaskTexture_SpeedY}, B=@n51.MaskTextureSpeedY_offset)
  n54 = TextureSampleParameter2D(UVs=Panner(Coordinate=Add(A=@n50, B=@n42), Speed=AppendVector(A=@n52, B=@n53)) {speedX=-1}) {
   group=MaskTexture
   parameterName=MaskTexture
   samplerType=SAMPLERTYPE_LinearColor
   texture=/BRCosmetics/Athena/Effects/Particles/Skydiving/NS_Contrail_StinkyCloud/T_Contrail_StinkyCloud_Ribbon
  }
  n55 = Multiply(A=Power(Base=Multiply(A=@n47.R, B=@n32), Exp=LinearInterpolate() {constA=2}), B=@n54.R)
  n56 = Power(Base=LinearInterpolate(Alpha=@n55) {constB=2.5}, Exp=ScalarParameter() {defaultValue=1, parameterName=Power})
  n57 = ScalarParameter() {defaultValue=-1, group=MaskTexture, parameterName=Orange_TextureSpeedX}
  n58 = ScalarParameter() {defaultValue=-0.05, group=MaskTexture, parameterName=Orange_TextureSpeedY}
  n59 = TextureSampleParameter2D(UVs=Panner(Coordinate=TextureCoordinate() {uTiling=0.8, vTiling=0.4}, Speed=AppendVector(A=@n57, B=@n58)) {speedX=-1, speedY=-0.05}) {parameterName=Orange_Texture, texture=/Game/Global/Textures/Default/Caustics/T_Caustics_Softer}
  n60 = Add(A=@n55, B=Multiply(A=@n56, B=Multiply(A=@n59.R) {constB=5}))
  n61 = ScalarParameter() {defaultValue=1, parameterName=Switch_Orange}
  n62 = LinearInterpolate(
   A=Multiply(A=@n22.A, B=@n32)
   B=Multiply(A=@n22.A, B=LinearInterpolate(A=@n60, B=@n56, Alpha=@n61))
   Alpha=ScalarParameter() {defaultValue=1, parameterName=Use_TexMask}
  )
  n63 = TextureCoordinate()
  n64 = Constant2Vector()
  n65 = ScalarParameter() {defaultValue=-2, group=Dissolve, parameterName=DissolveTexture_SpeedX}
  n66 = ScalarParameter() {group=Dissolve, parameterName=DissolveTexture_SpeedY}
  n67 = TextureSampleParameter2D(UVs=Panner(Coordinate=Add(A=@n63, B=@n64), Speed=AppendVector(A=@n65, B=@n66)) {speedX=-2}) {
   group=Dissolve
   parameterName=DissolveTexture
   samplerType=SAMPLERTYPE_LinearColor
   texture=&T_Contrail_StinkyCloud_Ribbon
  }
  n68 = DynamicParameter() {defaultValue=<rgba -1 1 1 1>, paramNames [Dissolve, MaskTextureSpeedX_offset, MaskTextureSpeedY_offset, SplineWidth_Dynamic]}
  n69 = Saturate(None=Subtract(A=@n67.R, B=@n68.Dissolve))
  n70 = Multiply(
   A=LinearInterpolate(A=Multiply(A=@n62, B=@n69), B=@n62, Alpha=ScalarParameter() {defaultValue=1, group=Dissolve, parameterName=Use_Disslove})
   B=MaterialFunctionCall("FadeSensitivity (optional)"=ScalarParameter() {parameterName=MeshFade}) {materialFunction=&FadeBasedOnViewAngle}.Result
  )
  n71 = DepthFade(Opacity=@n70, FadeDistance=ScalarParameter() {defaultValue=5, parameterName=DepthFade})
  n72 = ShadingPathSwitch(Default=@n71, Deferred=@n71, Forward=@n70, Mobile=@n70)
  n73 = WorldPosition()
  n74 = ParticlePositionWS()
  n75 = ScalarParameter() {defaultValue=500, parameterName=WorldPosition_SphereMaskRadius}
  LinearInterpolate(
   A=@n72
   B=Multiply(A=@n72, B=SphereMask(A=@n73.XYZ, B=@n74, Radius=@n75) {attenuationRadius=65, hardnessPercent=50})
   Alpha=ScalarParameter() {defaultValue=1, parameterName=Use_WorldPositionSphereMask}
  )
  n76 = LinearInterpolate(
   A=VectorParameter() {group=Color, parameterName=OutsideColor, defaultValue=<rgba 2 0 0 1>}.RGB
   B=VectorParameter() {group=Color, parameterName=CenterColor, defaultValue=<rgba 3 0.8121 0 1>}.RGB
   Alpha=Power(Base=@n32, Exp=ScalarParameter() {defaultValue=1, group=Color, parameterName=ColorLerpAlphaPower})
  )
  MaterialFunctionCall(Emissive=LinearInterpolate(A=Multiply(A=@n76, B=@n22.RGB), B=@n22.RGB, Alpha=ScalarParameter() {defaultValue=1, group=Color, parameterName=Use_ColorCurve})) {materialFunction=&MF_FX_EmissiveAdjustments}
 ]
}
