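# Niagara system for the SaladDressing emote's short, red-hot burst effect, with every particle living under a second.
# Five independent burst emitters layer spinning ring and smoke meshes, a shockwave, an energy sphere and a glow sprite, animated by size, color and material curves.
# Mesh scales divide out owner scale.
# Time-of-day, front-end and platform functions adjust brightness, opacity or size, and only the energy sphere uses distance culling.
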NiagaraSystem /BRCosmetics/Effects/Fort_Effects/Emotes/SaladDressing/FX/NS_Emote_SaladDressing_Burst
dependencies {
 usedDataInterfaces [&NiagaraDataInterfaceCurve, &NiagaraDataInterfaceColorCurve, &NiagaraDataInterfacePlatformSet]
 usedDynamicInputs []
 usedModules [&SystemState]
}
emitters {
 Burst_Ring {
  emitterUpdateScript [
   &EmitterState(Inactive Response=::NewEnumerator1, Life Cycle Mode=::NewEnumerator1, Loop Behavior=::NewEnumerator1, Loop Duration=1, UseLoopDelay=false)
   &SpawnBurst_Instantaneous(Spawn Count=4, Use Spawn Probability=false, UseLoopCountLimit=false)
  ]
  particleSpawnScript [
   &InitializeParticle(
    Color Mode=::NewEnumerator0
    Lifetime Max=0.8
    Lifetime Min=0.5
    Lifetime Mode=::NewEnumerator1
    Mesh Scale Max=<2 2 1.5>
    Mesh Scale Mode=::NewEnumerator2
    UsePositionOffset=false
    Write Material Random=true
   )
   &InitialMeshOrientation(Rotation=&RandomRangeVector(Minimum=<0 0 0>, Maximum=<1 1 1>), Use Rotation Vector=true)
  ]
  particleUpdateScript [
   &ParticleState()
   &UpdateMeshOrientation(Rotation Rate=3, Rotation Vector=<0 0 1>)
   &ScaleMeshSize(Scale Factor=&MakeVector(
     X=&FloatFromCurve(FloatCurve=DataInterface() {propertyValues {Curve {keys [<at 0 1>,<at 1 0.5>]}}})
     Y=&FloatFromCurve(FloatCurve=DataInterface() {propertyValues {Curve {keys [<at 0 1>,<at 1 0.5>]}}})
     Z=&FloatFromCurve(FloatCurve=DataInterface() {propertyValues {Curve {keys [<at 0 0.5>,<at 1 1.5>]}}})
    ))
   &ScaleMeshSize(Scale Factor=&Multiply_Vector(
     A=&VectorFromFloat(Value=&FloatFromCurve(FloatCurve=DataInterface() {propertyValues {Curve {keys [<at 0 0.5>,<at 1 2>]}}}, Scale Curve=0.7))
     B=&Divide_Vector(A:Vector3f=<1 1 1>, B:Vector3f=$Engine.Owner.Scale)
    )) {module=ScaleMeshSize001}
   &DynamicMaterialParameters(
    Index 0 Param 1=&FloatFromCurve(FloatCurve=DataInterface() {propertyValues {Curve {keys [
         {arriveTangent=-1.5522, interpMode=RCIM_Cubic, leaveTangent=-1.5522, tangentMode=RCTM_User, time=0, value=1}
         {arriveTangent=-0.9602, interpMode=RCIM_Cubic, leaveTangent=-0.9602, tangentMode=RCTM_User, time=1, value=0}
        ]}}})
    Module.Param0WriteEnabled=true
    Module.Param1WriteEnabled=true
    Module.Param2WriteEnabled=true
    Module.Param3WriteEnabled=true
   ) {module=DynamicMaterialParameters001}
   &ScaleColor(
    Scale Alpha=&Multiply_Float(
     A=&FloatFromCurve(
      FloatCurve=DataInterface() {propertyValues {Curve {keys [
          <at 0 0 RCIM_Cubic>
          {arriveTangent=-1.9777, interpMode=RCIM_Cubic, leaveTangent=-1.9777, tangentMode=RCTM_User, time=0.1, value=1}
          <at 1 0 RCIM_Cubic>
         ]}}}
      Scale Curve=&FNTimeOfDayScalar(Night Scale=0.3, Morning Scale=0.6, Evening Scale=0.3)
     )
     B=&FN_FrontEndFloatModifier(Gameplay=1, Front End=0.1)
    )
    Scale RGB=&MakeVectorFromLinearColorRGB(Color=&ColorFromCurve(ColorCurve=DataInterface() {propertyValues {
        AlphaCurve {keys []}
        BlueCurve {keys [
          {arriveTangent=-0.0154, interpMode=RCIM_Cubic, time=0.1561, value=1.4965}
          {arriveTangent=-0.1608, interpMode=RCIM_Cubic, leaveTangent=-0.1608, tangentMode=RCTM_User, time=0.385, value=0.1}
          {interpMode=RCIM_Cubic, leaveTangent=-0.0154, time=0.9205, value=0.0053}
         ]}
        GreenCurve {keys [
          {arriveTangent=-0.0154, interpMode=RCIM_Cubic, time=0.1561, value=1.3542}
          {arriveTangent=-0.0301, interpMode=RCIM_Cubic, leaveTangent=-0.0301, tangentMode=RCTM_User, time=0.385, value=0}
          {interpMode=RCIM_Cubic, leaveTangent=-0.0154, time=0.9205, value=0}
         ]}
        RedCurve {keys [
          {arriveTangent=-0.0154, interpMode=RCIM_Cubic, time=0.1561, value=10}
          {arriveTangent=-13.0346, interpMode=RCIM_Cubic, leaveTangent=-13.0346, time=0.385, value=5}
          {interpMode=RCIM_Cubic, leaveTangent=-0.0154, time=0.9205, value=0.0365}
         ]}
       }}))
    ScaleA=true
    ScaleRGB=true
   )
  ]
  rendererClasses [&NiagaraMeshRendererProperties]
  renderers [{rendererClass=&NiagaraMeshRendererProperties, rendererIndex=0}]
 }
 Burst_glow {
  emitterUpdateScript [
   &EmitterState(Inactive Response=::NewEnumerator1, Life Cycle Mode=::NewEnumerator1, Loop Behavior=::NewEnumerator1, Loop Duration=1, UseLoopDelay=false)
   &SpawnBurst_Instantaneous(Use Spawn Probability=false, UseLoopCountLimit=false)
  ]
  particleSpawnScript [
   &InitializeParticle(
    Color=<rgba 3 0.03 0 1>
    Lifetime=0.5
    Sprite Rotation Mode=::NewEnumerator2
    Sprite Size Mode=::NewEnumerator3
    Uniform Sprite Size=350
    UsePositionOffset=false
    Write Material Random=true
   )
   &CameraOffset(Camera Offset Amount=50)
  ]
  particleUpdateScript [
   &ParticleState()
   &ScaleSpriteSize(Uniform Curve Sprite Scale=DataInterface() {propertyValues {Curve {keys [
        {arriveTangent=-0.4932, interpMode=RCIM_Cubic, leaveTangent=-0.4932, tangentMode=RCTM_User, time=0, value=0.55}
        {arriveTangent=-3.0216, interpMode=RCIM_Cubic, leaveTangent=-3.0216, tangentMode=RCTM_User, time=0.05, value=1.5}
        <at 1 0 RCIM_Cubic>
       ]}}})
   &ScaleColor(
    Scale Alpha=&Multiply_Float(
     A=&FloatFromCurve(
      FloatCurve=DataInterface() {propertyValues {Curve {keys [<at 0 0>,<at 0.1 1>,<at 0.4 1>,<at 1 0>]}}}
      Scale Curve=&FNTimeOfDayScalar(Night Scale=0.3, Morning Scale=0.6, Evening Scale=0.3)
     )
     B=&FN_FrontEndFloatModifier(Gameplay=1, Front End=0.3)
    )
    Scale RGB=&VectorFromFloat(Value=&FN_PickFloatDependOnPlatform(TargetPlatformActive=1, TargetPlatformNoActive=0.4))
    ScaleA=true
    ScaleRGB=true
   )
  ]
  rendererClasses [&NiagaraSpriteRendererProperties]
  renderers [{rendererClass=&NiagaraSpriteRendererProperties}]
 }
 Energy_Sphere {
  emitterUpdateScript [
   &EmitterState(
    Enable Distance Culling=true
    Inactive Response=::NewEnumerator1
    Life Cycle Mode=::NewEnumerator1
    Loop Behavior=::NewEnumerator1
    Loop Duration=1
    Max Distance Response=::SleepAndClearParticles
    MaxDistance=2000
    Module.bUseMaxDistance=true
    Module.bUseMinDistance=false
    Scalability Mode=::NewEnumerator1
    UseLoopDelay=false
   )
   &SpawnBurst_Instantaneous(Spawn Count=1, Use Spawn Probability=true)
  ]
  particleSpawnScript [&InitializeParticle(
    Color=<rgba 1 0.8602 0.0743 0.08>
    Lifetime=&MakeCustomFloatFromBool(Bool=$User.bIsFrontEndPreview, True Float=0.4, False Float=0.45)
    Mesh Scale Mode=::NewEnumerator3
    Mesh Uniform Scale=3
    UsePositionOffset=false
    Write Material Random=true
   ) {module=InitializeParticle001}]
  particleUpdateScript [
   &ParticleState()
   /SetVariables(Particles.Dyn002:Vector4f=&MakeVector4(
     X=&FloatFromCurve(FloatCurve=DataInterface() {propertyValues {Curve {keys [<at 0 1>,<at 0.3 0.98>,<at 1 1>]}}})
     Y=&FloatFromCurve(FloatCurve=DataInterface() {propertyValues {Curve {keys [<at 0 15>,<at 1 0>]}}})
     Z=0
     W=0
    )) {isSetParameters=true}
   &ScaleColor(
    Scale Alpha=&FloatFromCurve(
     FloatCurve=DataInterface() {propertyValues {Curve {keys [{arriveTangent=-2.709, interpMode=RCIM_Cubic, leaveTangent=-2.709, tangentMode=RCTM_User, time=0.5, value=1}, <at 1 0 RCIM_Cubic>]}}}
     Scale Curve=&FNTimeOfDayScalar(Night Scale=10, Morning Scale=10, Day Scale=15, Evening Scale=10)
    )
    Scale RGB=&MakeVectorFromLinearColorRGB(Color=&ColorFromCurve(ColorCurve=DataInterface() {propertyValues {
        AlphaCurve {keys []}
        BlueCurve {keys [<at 0.4467 0.0787 RCIM_Cubic>,<at 0.9462 0.0241 RCIM_Cubic>]}
        GreenCurve {keys [<at 0.4467 0 RCIM_Cubic>,<at 0.9462 0 RCIM_Cubic>]}
        RedCurve {keys [<at 0.4467 1 RCIM_Cubic>,<at 0.9462 1 RCIM_Cubic>]}
       }}))
    ScaleA=true
    ScaleRGB=true
   )
   &ScaleColor(
    Scale Alpha=&FloatFromCurve(
     FloatCurve=DataInterface() {propertyValues {Curve {keys [{arriveTangent=-1.582, interpMode=RCIM_Cubic, leaveTangent=-1.582, tangentMode=RCTM_User, time=0, value=1}, <at 1 0 RCIM_Cubic>]}}}
     Scale Curve=&FN_FrontEndFloatModifier(Gameplay=1, Front End=0.5)
    )
    Scale RGB=&VectorFromFloat(Value=&FloatFromCurve(
      FloatCurve=DataInterface() {propertyValues {Curve {keys [<at 0 1 RCIM_Cubic>,<at 0.2 1 RCIM_Cubic>]}}}
      Scale Curve=&FN_FrontEndFloatModifier(Gameplay=1, Front End=0.5)
     ))
    ScaleA=true
    ScaleRGB=true
   ) {module=ScaleColor001}
   &ScaleMeshSize(Scale Factor=&Multiply_Vector(
     A=&VectorFromFloat(Value=&FloatFromCurve(
       FloatCurve=DataInterface() {propertyValues {Curve {keys [
           {arriveTangent=2.099, interpMode=RCIM_Cubic, leaveTangent=2.099, tangentMode=RCTM_User, time=0, value=0}
           {arriveTangent=1.0066, interpMode=RCIM_Cubic, leaveTangent=1.0066, tangentMode=RCTM_User, time=0.3439, value=0.5911}
           {arriveTangent=0.1607, interpMode=RCIM_Cubic, leaveTangent=0.1607, tangentMode=RCTM_User, time=1, value=1}
          ]}}}
       Scale Curve=&FN_FrontEndFloatModifier(Gameplay=1, Front End=1)
      ))
     B=&Divide_Vector(A:Vector3f=<1 1 1>, B:Vector3f=$Engine.Owner.Scale)
    ))
   &CameraOffset(Camera Offset Amount=&FloatFromCurve(FloatCurve=DataInterface() {propertyValues {Curve {keys [<at 0 -0.5>,<at 1 1>]}}}, Scale Curve=30))
  ]
  rendererClasses [&NiagaraMeshRendererProperties]
  renderers [{rendererClass=&NiagaraMeshRendererProperties, rendererIndex=0}]
 }
 Smoke_Ring {
  emitterUpdateScript [
   &EmitterState(Inactive Response=::NewEnumerator1, Life Cycle Mode=::NewEnumerator1, Loop Behavior=::NewEnumerator1, Loop Duration=1, UseLoopDelay=false)
   &SpawnBurst_Instantaneous(Spawn Count=3, Use Spawn Probability=false, UseLoopCountLimit=false)
  ]
  particleSpawnScript [
   &InitializeParticle(
    Lifetime Max=0.8
    Lifetime Min=0.6
    Lifetime Mode=::NewEnumerator1
    Mesh Scale=&Multiply_VectorByFloat(
     Vector=&MakeVector(X=1, Y=1, Z=&RandomRangeFloat(Minimum=1, Maximum=2))
     Float=&RandomRangeFloat(Minimum=0.3, Maximum=0.4)
    )
    Mesh Scale Mode=::NewEnumerator4
    UsePositionOffset=false
    Write Material Random=true
   ) {module=InitializeParticle001}
   &InitialMeshOrientation(
    Mesh Orientation Mode=::NewEnumerator2
    Orientation Coordinate Space=::Local
    Rotation=&RandomRangeVector(Minimum=<0 0 0>, Maximum=<0 0.5 1>)
    Use Rotation Vector=true
   ) {module=InitialMeshOrientation001}
  ]
  particleUpdateScript [
   &ParticleState()
   &UpdateMeshOrientation(Rotation Rate=&RandomRangeFloat(Minimum=-1, Maximum=-2), Rotation Vector=<0 0 1>)
   &ScaleMeshSize(Scale Factor=&Multiply_Vector(
     A=&VectorFromFloat(Value=2)
     B=&Divide_Vector(A:Vector3f=<1 1 1>, B:Vector3f=$Engine.Owner.Scale)
    )) {module=ScaleMeshSize001}
   &ScaleMeshSize(Scale Factor=&Multiply_VectorByFloat(
     Vector=&Multiply_VectorByFloat(
      Vector=&MakeVector(
       X=1
       Y=1
       Z=&Multiply_Float(
        A=&FloatFromCurve(
         FloatCurve=DataInterface() {propertyValues {Curve {keys [{arriveTangent=-1.572, time=0, value=1}, {interpMode=RCIM_Cubic, tangentMode=RCTM_User, time=1, value=0.5}]}}}
         Scale Curve=0.7
        )
        B=&RandomRangeFloat(Minimum=1, Maximum=1)
       )
      )
      Float=0.8
     )
     Float=&FloatFromCurve(
      FloatCurve=DataInterface() {propertyValues {Curve {keys [
          {arriveTangent=7.6944, interpMode=RCIM_Cubic, leaveTangent=7.6944, tangentMode=RCTM_User, time=0, value=0}
          {arriveTangent=3.5, interpMode=RCIM_Cubic, leaveTangent=3.5, time=0.2929, value=1.9515}
          {arriveTangent=0.6043, interpMode=RCIM_Cubic, leaveTangent=0.6043, tangentMode=RCTM_User, time=1, value=3.5}
         ]}}}
      Scale Curve=&FN_FrontEndFloatModifier(Gameplay=1.7, Front End=1.7)
     )
    ))
   &ScaleColor(
    Scale Alpha=&Multiply_Float(
     A=&FloatFromCurve(
      FloatCurve=DataInterface() {propertyValues {Curve {keys [
          {arriveTangent=6.8924, interpMode=RCIM_Cubic, leaveTangent=6.8924, tangentMode=RCTM_User, time=0, value=0}
          {arriveTangent=0.0028, interpMode=RCIM_Cubic, leaveTangent=0.0028, tangentMode=RCTM_User, time=0.2, value=1}
          {arriveTangent=-0.0149, interpMode=RCIM_Cubic, leaveTangent=-0.0149, tangentMode=RCTM_User, time=1, value=0}
         ]}}}
      Scale Curve=&FNTimeOfDayScalar(Night Scale=0.3, Morning Scale=0.6, Evening Scale=0.3)
     )
     B=1
    )
    Scale RGB=&MakeVectorFromLinearColorRGB(Color=&ColorFromCurve(ColorCurve=DataInterface() {propertyValues {
        AlphaCurve {keys []}
        BlueCurve {keys [
          {arriveTangent=-0.0154, interpMode=RCIM_Cubic, leaveTangent=-0.0154, tangentMode=RCTM_User, time=0.1561, value=0.7079}
          <at 0.3725 0.1>
          {arriveTangent=-0.0154, interpMode=RCIM_Cubic, leaveTangent=-0.0154, tangentMode=RCTM_User, time=0.6999, value=0.0053}
         ]}
        GreenCurve {keys [
          {arriveTangent=-0.0154, interpMode=RCIM_Cubic, leaveTangent=-0.0154, tangentMode=RCTM_User, time=0.1561, value=0.4688}
          <at 0.3725 0>
          {arriveTangent=-0.0154, interpMode=RCIM_Cubic, leaveTangent=-0.0154, tangentMode=RCTM_User, time=0.6999, value=0}
         ]}
        RedCurve {keys [
          {arriveTangent=-0.0154, interpMode=RCIM_Cubic, leaveTangent=-0.0154, tangentMode=RCTM_User, time=0.1561, value=15}
          <at 0.3725 5>
          {arriveTangent=-0.0154, interpMode=RCIM_Cubic, leaveTangent=-0.0154, tangentMode=RCTM_User, time=0.6999, value=0.0365}
         ]}
       }}))
    ScaleA=true
    ScaleRGB=true
   )
  ]
  rendererClasses [&NiagaraMeshRendererProperties]
  renderers [{rendererClass=&NiagaraMeshRendererProperties, rendererIndex=0}]
 }
 Wave {
  emitterUpdateScript [
   &EmitterState(Inactive Response=::NewEnumerator1, Life Cycle Mode=::NewEnumerator1, Loop Behavior=::NewEnumerator1, Loop Duration=1, UseLoopDelay=false)
   &SpawnBurst_Instantaneous(Spawn Count=1, Use Spawn Probability=false)
  ]
  particleSpawnScript [
   &InitializeParticle(
    Color=<rgba 5 5 5 1>
    Lifetime=0.4
    Mesh Scale Mode=::NewEnumerator3
    Mesh Uniform Scale=0.53
    UsePositionOffset=false
    Write Material Random=true
   )
   &InitialMeshOrientation(Orientation Axis=<0 0 1>, Orientation Vector=<0 0 1>, Rotation=<0 0.25 0>, Use Orientation Vector=true, Use Rotation Vector=true)
   &InitialMeshOrientation(
    Orientation Axis=<0 0 0>
    Orientation Vector=<0 0 0>
    Rotation=&RandomRangeVector(Minimum=<0 0 0>, Maximum=<1 0 0>)
    Use Orientation Vector=false
    Use Rotation Vector=true
   ) {module=InitialMeshOrientation001}
  ]
  particleUpdateScript [
   &ParticleState()
   &DynamicMaterialParameters(
    Index 0 Param 1=&FloatFromCurve(FloatCurve=DataInterface() {propertyValues {Curve {keys [<at 0 0.1>,<at 1 -0.4>]}}}, Scale Curve=0.8)
    Index 0 Param 2=&FloatFromCurve(FloatCurve=DataInterface() {propertyValues {Curve {keys [{arriveTangent=-0.3464, interpMode=RCIM_Cubic, leaveTangent=-0.3464, tangentMode=RCTM_User, time=0, value=1}, <at 1 0.9>]}}})
    Index 0 Param 3=&FloatFromCurve(FloatCurve=DataInterface() {propertyValues {Curve {keys [<at 0 0>,<at 0.15 1>,<at 1 1.2>]}}})
    Module.Param0WriteEnabled=true
    Module.Param1WriteEnabled=true
    Module.Param2WriteEnabled=true
    Module.Param3WriteEnabled=false
   )
   &ScaleMeshSize(Scale Factor=&Multiply_Vector(
     A=&VectorFromFloat(Value=&FloatFromCurve(
       FloatCurve=DataInterface() {propertyValues {Curve {keys [
           {arriveTangent=10.4172, interpMode=RCIM_Cubic, leaveTangent=10.4172, tangentMode=RCTM_User, time=0, value=0}
           {arriveTangent=3.1099, interpMode=RCIM_Cubic, leaveTangent=3.1099, tangentMode=RCTM_User, time=0.2, value=1.375}
           {arriveTangent=0.0906, interpMode=RCIM_Cubic, leaveTangent=0.0906, tangentMode=RCTM_User, time=1, value=2.5}
          ]}}}
       Scale Curve=&MakeCustomFloatFromBool(Bool=$User.bIsFrontEndPreview, True Float=0.8, False Float=1)
      ))
     B=&Divide_Vector(A:Vector3f=<1 1 1>, B:Vector3f=$Engine.Owner.Scale)
    ))
   &ScaleColor(
    Linear Color Curve=DataInterface() {propertyValues {
      AlphaCurve {keys [<at 0 1>,<at 0.5679 1>,<at 0.9849 0>]}
      BlueCurve {keys [<at 0.0618 0.4888>,<at 0.2291 0.0945>,<at 0.8773 0.0078>]}
      GreenCurve {keys [<at 0.0618 0.2292>,<at 0.2291 0.0469>,<at 0.8773 0.0078>]}
      RedCurve {keys [<at 0.0618 2>,<at 0.2291 1>,<at 0.8773 0.5>]}
     }}
    Scale Mode=::NewEnumerator2
   )
   &ScaleColor(
    Scale Alpha=&Multiply_Float(
     A=&FNTimeOfDayScalar(Night Scale=0.8, Morning Scale=0.8, Evening Scale=0.8)
     B=&FN_FrontEndFloatModifier(Gameplay=1, Front End=0.5)
    )
    Scale RGB=&VectorFromFloat(Value=&FloatFromCurve(FloatCurve=DataInterface() {propertyValues {Curve {keys [<at 0 2>,<at 0.18 1>,<at 1 1>]}}}))
    ScaleA=true
    ScaleRGB=true
   ) {module=ScaleColor001}
  ]
  rendererClasses [&NiagaraMeshRendererProperties]
  renderers [{rendererClass=&NiagaraMeshRendererProperties, rendererIndex=0}]
 }
}
summary {
 systemName=NS_Emote_SaladDressing_Burst
 emitters [
  {emitterName=Burst_Ring, rendererClasses [&NiagaraMeshRendererProperties]}
  {emitterName=Smoke_Ring, rendererClasses [&NiagaraMeshRendererProperties]}
  {emitterName=Burst_glow, rendererClasses [&NiagaraSpriteRendererProperties]}
  {emitterName=Wave, rendererClasses [&NiagaraMeshRendererProperties]}
  {emitterName=Energy_Sphere, rendererClasses [&NiagaraMeshRendererProperties]}
 ]
 userVariables [
  User.bIsFrontEnd {type=NiagaraBool}
  User.bIsFrontEndPreview {type=NiagaraBool}
  User.FrontEndSize {type=NiagaraFloat, defaultValue=0.6, description=""}
 ]
}
